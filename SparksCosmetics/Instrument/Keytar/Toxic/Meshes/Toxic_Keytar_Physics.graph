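# Physics asset for the Toxic Keytar instrument cosmetic, giving its mouth a jointed, physics-capable tongue.
# Four box bodies form the tongue chain from tongue_01 to tongue_04, rooted on a zero-length mouth capsule.
# Each link has its own swing limits, and the tip joint locks twist.
# Two capsules on jaw_upper add collision shapes for the upper jaw, which has no constraint to the chain.
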PhysicsAsset /SparksCosmetics/Instrument/Keytar/Toxic/Meshes/Toxic_Keytar_Physics
bodies {
 jaw_upper [
  {
   shapeType=Capsule
   length=3.7091
   radius=3.3006
   shapeName=jaw_upper_capsule
   center=<4.3565 -0.0617 0.0211>
   rotation=<rot -82.5607 126.6172 128.5053>
  }
  {
   shapeType=Capsule
   length=4.1835
   radius=3.7071
   shapeName=jaw_upper_capsule
   center=<7.1875 0 -0.9462>
   rotation=<rot -86.6413 179.9992 -179.9992>
  }
 ]
 mouth [{shapeType=Capsule, length=0, radius=2, shapeName=mouth_capsule}]
 tongue_01 [{shapeType=Box, extentX=2.5, extentY=3.5, extentZ=1.0167, shapeName=tongue_01, center=<-0.2034 0 0.3613>}]
 tongue_02 [{shapeType=Box, extentX=2.5, extentY=3, extentZ=1.0797, shapeName=tongue_02, center=<-0.0683 -0.1992 1.0172>, rotation=<rot -34.6129 0 0>}]
 tongue_03 [{
   shapeType=Box
   extentX=1.5
   extentY=3.5
   extentZ=0.8301
   shapeName=tongue_03
   center=<0.1244 0 0.3587>
   rotation=<rot 10.9592 -0.0011 -0.0058>
  }]
 tongue_04 [{
   shapeType=Box
   extentX=1.5
   extentY=3.5
   extentZ=1.5
   shapeName=tongue_03
   center=<0.3198 0.0001 -0.2603>
   rotation=<rot 34.9638 -0.0039 -0.0069>
  }]
}
constraints [
 {bone1Name=tongue_02, bone2Name=tongue_01, swing1LimitDegrees=40, swing2LimitDegrees=45, twistLimitDegrees=0}
 {bone1Name=tongue_03, bone2Name=tongue_02, swing1LimitDegrees=35, swing2LimitDegrees=30, twistLimitDegrees=10}
 {bone1Name=tongue_04, bone2Name=tongue_03, swing1LimitDegrees=35, swing2LimitDegrees=35, twistLimitDegrees=0, twistMotion=Locked}
 {bone1Name=tongue_01, bone2Name=mouth, swing1LimitDegrees=50, swing2LimitDegrees=30, twistLimitDegrees=0}
]
