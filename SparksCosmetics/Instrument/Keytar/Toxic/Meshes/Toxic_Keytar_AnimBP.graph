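# Animation Blueprint for the Toxic Keytar instrument cosmetic, adding twitching eyes and physics-driven secondary motion.
# The AnimGraph applies additive rotations to the upper and lower eye bones, then runs a RigidBody simulation based on the handle bone.
# The update event re-rolls eye rotations through random delays and switches simulation external force for skydiving and parachute states.
# Jaw rotation nodes exist but are unconnected.
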AnimBlueprint /SparksCosmetics/Instrument/Keytar/Toxic/Meshes/Toxic_Keytar_AnimBP
target_skeleton=&/Toxic_Keytar_Skeleton
graphs [
 AnimGraph {nodes [
   "Transform (Modify) Bone - Bone: jaw_upper"!(ComponentPose, Alpha=1, Rotation=Get JawRotator() -> (JawRotator:Rotator)) -> (Pose) {rotationMode=BMM_Additive, rotationSpace=BCS_BoneSpace, alphaBoolBlend {blendOption=Linear}, boneToModify {boneName=jaw_upper}}
   n1 = "Transform (Modify) Bone - Bone: eye_upper"!(ComponentPose=Local To Component(LocalPose=Input Pose()), Alpha=1, Rotation=Get UpperEyeRotator() -> (UpperEyeRotator:Rotator)) -> (Pose) {rotationMode=BMM_Additive, rotationSpace=BCS_BoneSpace, alphaBoolBlend {blendOption=Linear}, boneToModify {boneName=eye_upper}}
   n2 = "Transform (Modify) Bone - Bone: eye_bottom"!(ComponentPose=@n1, Alpha=1, Rotation=Get LowerEyeRotator() -> (LowerEyeRotator:Rotator)) -> (Pose) {rotationMode=BMM_Additive, rotationSpace=BCS_BoneSpace, alphaBoolBlend {blendOption=Linear}, boneToModify {boneName=eye_bottom}}
   n3 = RigidBody(ComponentPose=@n2, ExternalForce=Get GravityOverride()) {
    bOverrideWorldGravity=true
    overridePhysicsAsset=&/Toxic_Keytar_Physics
    simulationSpace=BaseBoneSpace
    alphaBoolBlend {blendOption=Linear}
    baseBoneRef {boneName=handle}
    componentLinearAccScale=<0.4 0.4 0.4>
    componentLinearVelScale=<0.2 0.2 0.2>
    overrideWorldGravity {z=-980}
   }
   Output Pose(Result=Component To Local(ComponentPose=@n3)) {layerGroup=None}
  ]}
 EventGraph {nodes [
   Event BlueprintUpdateAnimation()
   > n1 = Sequence()
   > n2 = Branch(Condition=Get bIsSkyDiving())
   > n3 = Branch(Condition=Get bIsParachuteOpened())
   > Set GravityOverride(GravityOverride="0,0,-400")
   Set GravityOverride(execute=@n3.else, GravityOverride="0,0,20000")
   Set GravityOverride(execute=@n2.else, GravityOverride="0,0,-980")
   Delay(execute=@n1.then_1, Duration=RandomFloatInRange(Min=0.25, Max=1.25))
   > Set!(
    execute
    UpperEyeRotator_Roll:float=RandomFloatInRange(Min=-10, Max=15)
    UpperEyeRotator_Pitch:float=0
    UpperEyeRotator_Yaw:float=RandomFloatInRange(Min=-10, Max=15)
   ) -> (then, Output_Get:Rotator)
   > Delay(Duration=RandomFloatInRange(Min=0.5, Max=1.5))
   > Set!(
    execute
    LowerEyeRotator_Roll:float=RandomFloatInRange(Min=-10, Max=10)
    LowerEyeRotator_Pitch:float=0
    LowerEyeRotator_Yaw:float=RandomFloatInRange(Min=-10, Max=10)
   ) -> (then, Output_Get)
   > Delay(Duration=2)
   TryGetPawnOwner()
   Set!(execute, JawRotator_Roll:float=0, JawRotator_Pitch:float=RandomFloatInRange(Min=-10, Max=15), JawRotator_Yaw:float=0) -> (then, Output_Get)
  ]}
]
variables [
 GravityOverride {type=Vector}
 JawRotator {type=Rotator}
 LowerEyeRotator {type=Rotator}
 UpperEyeRotator {type=Rotator}
]
